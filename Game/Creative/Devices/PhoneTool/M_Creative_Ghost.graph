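# Additive unlit particle material for the Creative phone tool, named as a ghost effect, that projects a glowing grid onto scene geometry behind the particle.
# The move-tool transition texture is world-aligned using the behind-translucency position converted to local space, and blends background and grid colors.
# Depth-based distance falloff and a UV sphere mask fade it before particle color tinting.
# Only the deferred shading path receives the result; others get black.
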Material /Game/Creative/Devices/PhoneTool/M_Creative_Ghost
graph {
 properties {
  bEnableResponsiveAA=true
  bUsedWithParticleSprites=true
  blendMode=BLEND_Additive
  pixelDepthOffsetMode=PDOM_Legacy
  shadingModel=MSM_Unlit
  usedShadingModels=Unlit
 }
 nodes [
  n1 = Constant()
  n2 = VectorParameter() {parameterName=BG Color, defaultValue=<rgba 0 0.0085 0.1372 0>}
  n3 = VectorParameter() {parameterName=GridColor, defaultValue=<rgba 0.1457 0.398 0.9427 0>}
  n4 = MaterialFunctionCall() {materialFunction=&WorldPositionBehindTranslucency}
  n5 = VertexNormalWS()
  n6 = ComponentMask(None=MaterialFunctionCall(
    TextureObject=TextureObject() {texture=/Game/Weapons/Prototype/ObjectMover/Textures/T_Movetool_Transition}
    TextureSize=VectorParameter() {parameterName=TextureSize, defaultValue=<rgba 24 24 24 0>}.RGB
    WorldPosition=TransformPosition(None=@n4.Result) {transformSourceType=TRANSFORMPOSSOURCE_World}
    World Space Normal=Transform(None=@n5) {transformSourceType=TRANSFORMSOURCE_World, transformType=TRANSFORM_Local}
   ) {materialFunction=&WorldAlignedTexture}.XYZ Texture) {r=true}
  n7 = ScalarParameter() {defaultValue=40, parameterName=Distance Divisor}
  n8 = Abs(None=ConstantBiasScale(None=TextureCoordinate()) {bias=-0.5, scale=2})
  n9 = Constant()
  n10 = Multiply(
   A=Multiply(
    A=LinearInterpolate(A=@n2.RGB, B=@n3.RGB, Alpha=@n6)
    B=MaterialFunctionCall!(Distance Bias=@n7, Distance Divisor=@n7, Scene Depth, Pixel Depth) {materialFunction=/Game/Effects/Fort_Effects/Materials/_Functions/MF_DistanceBehindTranslucency_01}.Normalized Inverted
   )
   B=SphereMask(A=Power(Base=@n8), B=@n9, Radius=Constant() {r=1}, Hardness=@n9)
  )
  n11 = ParticleColor()
  MaterialFunctionCall(Emissive=ShadingPathSwitch(Default=@n1, Deferred=Multiply(A=@n10, B=@n11.RGB), Forward=@n1, Mobile=@n1)) {materialFunction=&MF_FX_EmissiveAdjustments}
 ]
}
